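; Parent gameplay ability that cancels the Leap Glider active power on the owning Fort player pawn.
; On activation it acts only if the pawn lacks the cancel cooldown tag, has an active glider tag and is skydiving; otherwise it just ends.
; Cancelling ends skydiving, sets falling movement, removes active-tag effects and unblocks equipment-change abilities.
; Removing the infinite cooldown effect lets the actual cooldown start.
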(event Ability|EventActivateAbility
  (bind _asfort_player_pawn (Utilities|Casting|CastToFortPlayerPawn (Ability|GetAvatarActorfromActorInfo))
    (:then
      (if (GameplayTags|HasMatchingGameplayTag _asfort_player_pawn (Variables|Default|GetT_CancelCooldown))
        _
        (elif (and (GameplayTags|HasAnyMatchingGameplayTags _asfort_player_pawn (Variables|Default|GetTC_ActiveTag)) (Parachute|IsSkydiving _asfort_player_pawn))
          (CallFunction|RanAbilitySuccesfully))))
    (:CastFailed
      _))
  (Ability|EndAbility) ; Failsafe to do nothing and end the ability if pawn isn't skydiving
)

(event Custom|RanAbilitySuccesfully
  ;[1 1 1] Ending skydiving stuff in here when cancelled - having issues sending events.
  (Parachute|EndSkydiving (stale Ability|EventActivateAbility _asfort_player_pawn))
  (Pawn|Components|CharacterMovement|SetMovementMode (Variables|/Script/Engine.Character|GetCharacterMovement (stale Ability|EventActivateAbility _asfort_player_pawn)) "MOVE_Falling" 0)
  (Ability|RemoveGameplayEffectFromOwnerWithGrantedTags (Variables|Default|GetTC_ActiveTag))
  (Tags|UnBlockAbilitiesWithTags (Ability|GetFortAbilitySystemComponent (Ability|GetAvatarActorfromActorInfo)) (Variables|Default|GetT%20ChangeEquipment) false)
  (Ability|RemoveGameplayEffectFromOwnerWithAssetTags (Variables|Default|GetTC_InfCooldown)) ; Removing this starts the actual cooldown
  (Ability|EndAbility))
