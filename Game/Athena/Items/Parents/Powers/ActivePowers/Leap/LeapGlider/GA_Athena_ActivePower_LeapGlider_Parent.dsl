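; Parent active-power ability that leaps the player into the sky and redeploys a glider with a cosmetic override.
; Activation runs Target and Leap, applies active and minimum-duration effects, and a montage notify force-opens the glider after a net sync.
; Rift, vent, landing, movement-mode, unequip and overflow-drop failsafes lead to shared cleanup, which starts cooldown by removing the infinite cooldown effect.
; Net syncs let the client clean up after server-only drop events.
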(event Custom|BoardDroppedMidGlide
  (params (ItemDropped :type Fort%20World%20Item%20Object%20Reference))
  (bind _returnvalue (Ability|GetAvatarActorfromActorInfo))
  (bind _returnvalue_2 (Fortnite|GetFortPlayerControllerfromActor _returnvalue))
  (bind _asfort_player_controller_athena (Utilities|Casting|CastToFortPlayerControllerAthena _returnvalue_2))
  ;[1 1 1] Fix for a bug that could happen if the board is activated while an item is about to come into your inventory.  Event only fires on the server, need to do this to make sure the client gets it and cleans up correctly
  (Ability|Tasks|WaitNetSync "OnlyClientWait"
    (:then
      (EventDispatchers|BindEventtoOnItemDroppedDuetoOverflow _asfort_player_controller_athena (delegate Custom|BoardDroppedMidGlide))
      (|CallParentFunction))
    (:OnSync
      (EventDispatchers|UnbindEventfromOnItemDroppedDuetoOverflow _asfort_player_controller_athena (delegate Custom|BoardDroppedMidGlide))
      (CallFunction|CleanupAbility))))

(event Custom|HandleItemEquip
  (bind _returnvalue (Ability|GetAvatarActorfromActorInfo))
  (bind _returnvalue_2 (Fortnite|GetFortPlayerControllerfromActor _returnvalue))
  (bind _asfort_player_controller_athena (Utilities|Casting|CastToFortPlayerControllerAthena _returnvalue_2))
  (FortPawn|BindEventtoOnWeaponUnEquippedDelegate (Utilities|Casting|CastToFortPlayerPawnAthena (Ability|GetAvatarActorfromActorInfo)) (EventDispatchers|CreateEvent :!event Custom|WeaponUnequipped))
  ;[1 1 1] Fix for a bug that could happen if the board is activated while an item is about to come into your inventory.  Event only fires on the server, need to do this to make sure the client gets it and cleans up correctly
  (Ability|IsLocallyControlled
    (:True
      (Ability|Tasks|WaitNetSync "OnlyClientWait"
        (:then
          _)
        (:OnSync
          (EventDispatchers|UnbindEventfromOnItemDroppedDuetoOverflow _asfort_player_controller_athena (delegate Custom|BoardDroppedMidGlide))
          (CallFunction|CleanupAbility))))
    (:False
      _))
  (EventDispatchers|BindEventtoOnItemDroppedDuetoOverflow _asfort_player_controller_athena (delegate Custom|BoardDroppedMidGlide))
  (|CallParentFunction))

(event Custom|WeaponUnequipped
  (CallFunction|CleanupAbility))

(event Ability|EventActivateAbility
  ;[1 1 1] ActivateAbility->Send us into the sky, and then deploy the soy glider
  (|CallParentFunction)
  (Utilities|PowerCommitAbility (Variables|AbilityCommit|GetIgnoreAbilityCommit)
    (:Pass
      _)
    (:Fail
      (Development|PrintString "Power: Ability Commit Failed")
      (CallFunction|PowerCommitFailed))
    (:Ignored
      _))
  (Utilities|FlowControl|Sequence
    (:then_0
      ;[1 1 1] Failsafe states - situations we'd encounter during gameplay that we'd need to clean the ability up for
      (Ability|Tasks|WaitGameplayTagAdd "(TagName=\"Gameplay.Status.Rift.Skydiving\")"
        (:Added
          (Ability|Tasks|WaitNetSync "OnlyServerWait"
            (:OnSync
              _)) ; Server gets here first with states applied via GE
        )))
    (:then_1
      (Ability|Tasks|WaitGameplayEvent "(TagName=\"Abilities.Generic.Vent.Enter\")"
        (:EventReceived
          (Ability|Tasks|WaitNetSync "OnlyClientWait"
            (:OnSync
              _)) ; Client gets here first with stuff that launches via event
        )))
    (:then_2
      (Ability|Tasks|WaitGameplayTagAdd "(TagName=\"Athena.Player.InVent\")"
        (:Added
          (Variables|PowerLeapGlider|SetGenericFailsafeHit true) ; Kill gliding if we hit a failsafe condition
          (CallFunction|GenericFailSafe)
          (Parachute|ForceOpenParachuteandOverrideGlider (Variables|Default|GetPlayerPawn) (Variables|PowerLeapGlider|GetGlider%20Cosmetic%20Override) true true))))
    (:then_3
      (CallFunction|Target) ; Break da buildin's
      (HUD|SetHUDElementVisibility (BlueprintContext|GetFortHUDContext) (Variables|PowerLeapGlider|GetT_SkyDivingMessageTag) true "None") ; Hide skydiving message
      (CallFunction|Leap)
      (Ability|ApplyGameplayEffectToOwner (Variables|PowerLeapGlider|GetGE_Active)) ; This is a tag for the cancel ability to check so that we dont just end any skydiving ability :D
      ;;[1 1 1] Set the min time the power is active before it can be canceled
      (bind _returnvalue (Ability|GameplayEffect|SetDuration (Ability|MakeOutgoingGameplayEffectSpec (Variables|PowerLeapGlider|GetGE_MinActiveTime)) (ScalableFloat|GetValueatLevel (Variables|PowerLeapGlider|GetRow_MinActiveTime))))
      (Ability|ApplyGameplayEffectSpecToOwner _returnvalue))
    (:then_4
      (Fortnite|FortShippingLog self (Utilities|String|Append "Running GA_Athena_ActivePower_LeapGlider_Parent with glider override: " (Utilities|GetDisplayName (Variables|PowerLeapGlider|GetGlider%20Cosmetic%20Override))))))
  (Variables|PowerLeapGlider|SetGenericFailsafeHit true) ; Kill gliding if we hit a failsafe condition
  (CallFunction|GenericFailSafe))

(event EventOnLand
  (params (Hit :type Hit%20Result%20Structure%20%28by%20ref%29))
  ;[1 1 1] Clean up leap feedback and ability if we land, our movement mode changes while we're gliding, or we hit a Rift.
  (|CallParentFunction Hit)
  (CallFunction|CleanupAbility))

(event Custom|MovementModeChangeFailSafe
  (params (Character :type Character%20Object%20Reference) (PrevMovementMode :type EMovementMode%20Enum :default "MOVE_None") (PreviousCustomMode :type Byte))
  ;[1 1 1] Clean up leap feedback and ability if we land, our movement mode changes while we're gliding, or we hit a Rift.
  (Character|UnbindEventfromMovementModeChangedDelegate Character (delegate Custom|MovementModeChangeFailSafe))
  (|CallParentFunction (stale EventOnLand Hit))
  (CallFunction|CleanupAbility))

(event Custom|GenericFailSafe
  ;[1 1 1] Clean up leap feedback and ability if we land, our movement mode changes while we're gliding, or we hit a Rift.
  (|CallParentFunction (stale EventOnLand Hit))
  (CallFunction|CleanupAbility))

(event Custom|CleanupAbility
  (bind _playerpawn (Variables|Default|GetPlayerPawn))
  ;[1 1 1] Ability cleanup - start cooldown, remove FX, return skydiving prompt to hide and a failsafe to end skydiving
  (Actor|DestroyActor (Utilities|Casting|CastToActor (Variables|PowerLeapGlider|GetLensEffectInterface)))
  (if (not (Vortex|IsinVortex (Variables|Default|GetPlayerPawn)))
    (Parachute|EndSkydiving (Variables|Default|GetPlayerPawn)) ; Failsafe
  )
  (Ability|RemoveGameplayEffectFromOwnerWithAssetTags (Variables|PowerLeapGlider|GetTC_InfCooldown)) ; Removing this starts the actual cooldown
  (if (GameplayTags|HasAnyMatchingGameplayTags (Variables|Default|GetPlayerPawn) (Variables|PowerLeapGlider|GetTC_GrantedActive))
    (Ability|RemoveGameplayEffectFromOwnerWithGrantedTags (Variables|PowerLeapGlider|GetTC_GrantedActive)))
  (HUD|SetHUDElementVisibility (BlueprintContext|GetFortHUDContext) (Variables|PowerLeapGlider|GetT_SkyDivingMessageTag) false "None")
  (FortPawn|UnbindEventfromOnWeaponUnEquippedDelegate _playerpawn (EventDispatchers|CreateEvent :!event Custom|WeaponUnequipped))
  (Tags|UnBlockAbilitiesWithTags (Ability|GetFortAbilitySystemComponent _playerpawn) (Variables|PowerLeapGlider|GetBlockedTags) true)
  (Development|PrintString "Cleanup Ability Event Ran" false false)
  (Ability|EndAbility))

(event EventCastMontageBlendOut)

(event EventCastMontageNotifyEndAbility)

(event EventCastMontageNotifyStart
  (params (NotifyTag :type Gameplay%20Tag%20Structure))
  (bind _playerpawn (Variables|Default|GetPlayerPawn))
  ;[1 1 1] We play a montage to fuse the JumpPad anims with actual glider redeploy.  This happens here.
  (if (and (GameplayTags|Equal(GameplayTag) NotifyTag (Variables|PowerLeapGlider|GetT_StartRedeploy)) (not (Variables|PowerLeapGlider|GetGenericFailsafeHit)))
    (Ability|Tasks|WaitNetSync "OnlyClientWait"
      (:OnSync
        (Variables|PowerLeapGlider|SetParachuteOpened true)
        (Utilities|FlowControl|Sequence
          (:then_0
            (Tags|UnBlockAbilitiesWithTags (Ability|GetFortAbilitySystemComponent (Variables|Default|GetPlayerPawn)) (Variables|PowerLeapGlider|GetT_OpenParahcute) true))
          (:then_1
            ;; PlayGliderOpen
            (Audio|SpawnSoundAttached (Variables|PowerLeapGlider|GetOpenGliderSoundCue) (Variables|/Script/Engine.Actor|GetRootComponent _playerpawn))
            (Parachute|ForceOpenParachuteandOverrideGlider _playerpawn (Variables|PowerLeapGlider|GetGlider%20Cosmetic%20Override) true true)
            (Character|BindEventtoMovementModeChangedDelegate _playerpawn (EventDispatchers|CreateEvent :!event Custom|MovementModeChangeFailSafe))))))
    (else
      (|CallParentFunction NotifyTag))))

(event EventCastMontageCancelledOrInterrupted)

(event EventPowerCommitFailed)

(event Ability|EventOnEndAbility
  (params (bWasCancelled :type Boolean))
  ;[1 1 1] Event EndAbility-> Failsafe for being dropped due to overflow. Dropped due to overflow delegate doesn't seem to fire.  Basically if it's dropped before glider redeploy, make it work still
  (if (Variables|PowerLeapGlider|GetParachuteOpened)
    (Variables|PowerLeapGlider|SetParachuteOpened false)
    (else
      (Ability|RemoveGameplayEffectFromOwnerWithAssetTags (Variables|PowerLeapGlider|GetTC_InfCooldown)) ; Removing this starts the actual cooldown
      (Ability|RemoveGameplayEffectFromOwnerWithGrantedTags (Variables|PowerLeapGlider|GetTC_GrantedActive))))
  (|CallParentFunction bWasCancelled))

(fn InitAbility
  (bind _lenseffectinterfaceclass (Variables|PowerLeapGlider|GetLensEffectInterfaceClass))
  (|CallParentFunction)
  (Variables|PowerLeapGlider|SetGenericFailsafeHit false)
  (LensEffect|IsValidCameraLensClass _lenseffectinterfaceclass
    (:Valid
      (bind _returnvalue (Feedback|SpawnGenericCameraLensEffect (Variables|Default|GetPlayerPawn) (LensEffect|GetInterfaceClass _lenseffectinterfaceclass)))
      (Variables|PowerLeapGlider|SetLensEffectInterface _returnvalue))))
